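# Sound cue for the impact sound of the Vampire pickaxe cosmetic, classed under Fort_Impacts.
# A Modulator feeds a quality-level switch, which picks among five random impact waves, a random set of three, or a single repeated wave for the remaining two tiers.
# It primes on load and uses shared pickaxe impact attenuation, a pickaxe impact property sheet and the player melee weapon concurrency set.
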SoundCue /BRCosmetics/Sounds/PickAxes/Vampire/PA_Vampire_Impact_Cue
root=Modulator_0
nodes [
 {id=Modulator_0, children [QualityLevel_0]}
 {id=QualityLevel_0, children [Random_0, Random_2, WavePlayer_14, WavePlayer_14]}
 {id=Random_0, children [WavePlayer_0,WavePlayer_1,WavePlayer_2,WavePlayer_4,WavePlayer_5], properties {weights [1,1,1,1,1]}}
 {id=Random_2, children [WavePlayer_11, WavePlayer_12, WavePlayer_13], properties {weights [1,1,1]}}
 {id=WavePlayer_0, properties {soundWaveAssetPtr=&/PA_Vampire_Impact_01}}
 {id=WavePlayer_1, properties {soundWaveAssetPtr=&/PA_Vampire_Impact_02}}
 {id=WavePlayer_11, properties {soundWaveAssetPtr=&PA_Vampire_Impact_01}}
 {id=WavePlayer_12, properties {soundWaveAssetPtr=&/PA_Vampire_Impact_03}}
 {id=WavePlayer_13, properties {soundWaveAssetPtr=&PA_Vampire_Impact_02}}
 {id=WavePlayer_14, properties {soundWaveAssetPtr=&PA_Vampire_Impact_01}}
 {id=WavePlayer_2, properties {soundWaveAssetPtr=&PA_Vampire_Impact_03}}
 {id=WavePlayer_4, properties {soundWaveAssetPtr=&/PA_Vampire_Impact_04}}
 {id=WavePlayer_5, properties {soundWaveAssetPtr=&/PA_Vampire_Impact_05}}
]
properties {
 attenuationSettings=/Game/Sounds/Attenuation/Fort_Pickaxe_Impact_Att
 audioPropertiesSheet=/BRCosmetics/Sounds/Settings/PropertySheets/Pickaxe/APS_Cosmetic_Pickaxe_Impact
 bPrimeOnLoad=true
 maxDistance=4300
 soundClassObject=&Fort_Impacts
 volumeMultiplier=0.6375
 attenuationOverrides {
  bAttenuateWithLPF=true
  distanceAlgorithm=Logarithmic
  falloffDistance=800
  lPFRadiusMax=800
  lPFRadiusMin=400
  attenuationShapeExtents {x=300}
 }
 concurrencyOverrides {bLimitToOwner=true, maxCount=2, resolutionRule=StopOldest}
 concurrencySet [&Player_Melee_Weapon_Conc]
}
